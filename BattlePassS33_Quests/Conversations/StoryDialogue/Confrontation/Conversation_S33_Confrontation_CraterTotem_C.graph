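# Conversation database for the Crater Totem "C" dialogue in the Season 33 Confrontation story quest line.
# A tagged entry point leads to a single terminal speech node with a main-quest icon and a line about a demonic force removing something powerful.
# The node carries an "Investigate the Crater" choice with its own quest conversation tag, and requires quest_S33_story_confrontation_q02 to be active.
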ConversationDatabase /BattlePassS33_Quests/Conversations/StoryDialogue/Confrontation/Conversation_S33_Confrontation_CraterTotem_C
entry_points [{tag=Conversation.Entry.Season33.Story.Confrontation.02.C, destinations [#1]}]
nodes [
 ConversationEntryPointNode_0 {
  id=#1
  class=ConversationEntryPointNode
  outputs [#2]
  properties {entryTag {tagName=Conversation.Entry.Season33.Story.Confrontation.02.C}}
 }
 Speech_C_0 {
  id=#2
  class=Speech_C
  outputs []
  properties {
   softTaskIcon=&T_UI_NPC_MainQuest
   generalConfig {defaultMessage=A demonic force came here and removed something powerful.}
  }
  sub_nodes [
   ConversationChoiceNode_0 {
    class=ConversationChoiceNode
    properties {
     defaultChoiceDisplayText=Investigate the Crater
     choiceTags {gameplayTags [{tagName=Athena.Quests.S33.Conversation.Story.Confrontation.02.CraterTotem.C}]}
    }
   }
   FortConversationRequirement_ControllerRequirement_0 {
    class=FortConversationRequirement_ControllerRequirement
    properties {
     nodeName=Require: Is Quest Active
     requirement {
      class=FortControllerRequirement_IsQuestActive
      properties {questDefinition=/BattlePassS33_Quests/Items/QuestItems/Story/S33_Story_Confrontation/quest_S33_story_confrontation_q02}
     }
    }
   }
  ]
 }
]
