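# Conversation database for the first Season 33 Confrontation story quest, where the player speaks to Kendo about Hope.
# A single entry point leads to a linear chain of three speech lines linking Hope's search, magic, the comet and freeing Jade.
# The opening main-questline choice requires the quest to be active.
# Continue choices advance the rest, with the last carrying the quest's conversation gameplay tag.
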ConversationDatabase /BattlePassS33_Quests/Conversations/StoryDialogue/Confrontation/Conversation_S33_Confrontation_Q01
entry_points [{tag=Conversation.Entry.Season33.Story.Confrontation.01, destinations [#1]}]
nodes [
 ConversationEntryPointNode_0 {
  id=#1
  class=ConversationEntryPointNode
  outputs [#2]
  properties {entryTag {tagName=Conversation.Entry.Season33.Story.Confrontation.01}}
 }
 Speech_C_3 {
  id=#2
  class=Speech_C
  outputs [#3]
  properties {
   softTaskIcon=&T_UI_NPC_MainQuest
   generalConfig {defaultMessage=I suspect whatever Hope is searching for is entangled in the emergence of magic.}
  }
  sub_nodes [
   ConversationChoiceNode_5 {
    class=ConversationChoiceNode
    properties {defaultChoiceDisplayText=Speak to Kendo about Hope, choiceTags {gameplayTags [{tagName=Conversation.Choice.Category.MainQuestline}]}}
   }
   FortConversationRequirement_ControllerRequirement_4 {
    class=FortConversationRequirement_ControllerRequirement
    properties {
     nodeName=Require: Is Quest Active
     requirement {
      class=FortControllerRequirement_IsQuestActive
      properties {questDefinition=/BattlePassS33_Quests/Items/QuestItems/Story/S33_Story_Confrontation/quest_S33_story_confrontation_q01}
     }
    }
   }
  ]
 }
 Speech_C_28 {
  id=#3
  class=Speech_C
  outputs [#4]
  properties {generalConfig {defaultMessage=It all began when that strange comet landed on the Island.}}
  sub_nodes [ConversationChoiceNode_6 {class=ConversationChoiceNode, properties {defaultChoiceDisplayText=Continue}}]
 }
 Speech_C_1 {
  id=#4
  class=Speech_C
  outputs []
  properties {generalConfig {defaultMessage=I wonder if that comet holds the secret to freeing Jade.}}
  sub_nodes [ConversationChoiceNode_8 {
    class=ConversationChoiceNode
    properties {defaultChoiceDisplayText=Continue, choiceTags {gameplayTags [{tagName=Athena.Quests.S33.Conversation.Story.Confrontation.01}]}}
   }]
 }
]
