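# Conversation database for the Season 33 Confrontation story quest step 06, where an NPC reacts to the player delivering the Splinter Blades.
# The entry tag leads to a main-questline choice, gated on quest q06c being active, followed by a linear chain of three speech lines advanced by Continue choices.
# The final Continue choice carries the Confrontation.06 quest tag, presumably to signal quest progress.
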ConversationDatabase /BattlePassS33_Quests/Conversations/StoryDialogue/Confrontation/Conversation_S33_Confrontation_Q06
entry_points [{tag=Conversation.Entry.Season33.Story.Confrontation.06, destinations [#1]}]
nodes [
 ConversationEntryPointNode_0 {
  id=#1
  class=ConversationEntryPointNode
  outputs [#2]
  properties {entryTag {tagName=Conversation.Entry.Season33.Story.Confrontation.06}}
 }
 Speech_C_2 {
  id=#2
  class=Speech_C
  outputs [#3]
  properties {
   softTaskIcon=&T_UI_NPC_MainQuest
   generalConfig {defaultMessage="These blades, they carry the comet's strange power."}
  }
  sub_nodes [
   ConversationChoiceNode_0 {
    class=ConversationChoiceNode
    properties {defaultChoiceDisplayText=Deliver the Splinter Blades, choiceTags {gameplayTags [{tagName=Conversation.Choice.Category.MainQuestline}]}}
   }
   FortConversationRequirement_ControllerRequirement_0 {
    class=FortConversationRequirement_ControllerRequirement
    properties {
     nodeName=Require: Is Quest Active
     requirement {
      class=FortControllerRequirement_IsQuestActive
      properties {questDefinition=/BattlePassS33_Quests/Items/QuestItems/Story/S33_Story_Confrontation/quest_S33_story_confrontation_q06c}
     }
    }
   }
  ]
 }
 Speech_C_28 {
  id=#3
  class=Speech_C
  outputs [#4]
  properties {generalConfig {defaultMessage="This could free me... and, perhaps, your Jade."}}
  sub_nodes [ConversationChoiceNode_6 {class=ConversationChoiceNode, properties {defaultChoiceDisplayText=Continue}}]
 }
 Speech_C_1 {
  id=#4
  class=Speech_C
  outputs []
  properties {generalConfig {defaultMessage=I know who to give this to. You'll have to trust me on this.}}
  sub_nodes [ConversationChoiceNode_8 {
    class=ConversationChoiceNode
    properties {defaultChoiceDisplayText=Continue, choiceTags {gameplayTags [{tagName=Athena.Quests.S33.Conversation.Story.Confrontation.06}]}}
   }]
 }
]
