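# Conversation database for a Season 33 story quest in which a condescending NPC challenges the player to a riddle.
# The opening Inquire choice requires the Confrontation q04a quest to be active, then a linear speech chain leads to the riddle.
# Three answer choices, ordered by sort-priority tags, end the conversation with a reaction line.
# Glider carries the quest Success tag, while Sun and Supply Drop carry the Failure tag.
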ConversationDatabase /BattlePassS33_Quests/Conversations/StoryDialogue/Confrontation/Conversation_S33_Confrontation_Riddle_A
entry_points [{tag=Conversation.Entry.Season33.Story.Confrontation.Riddle.01, destinations [#1]}]
nodes [
 ConversationEntryPointNode_1 {
  id=#1
  class=ConversationEntryPointNode
  outputs [#2]
  properties {entryTag {tagName=Conversation.Entry.Season33.Story.Confrontation.Riddle.01}}
 }
 Speech_C_2 {
  id=#2
  class=Speech_C
  outputs [#3]
  properties {
   softTaskIcon=&T_UI_NPC_MainQuest
   generalConfig {defaultMessage="Do you even comprehend the power you stand before? Now, come amuse me."}
  }
  sub_nodes [
   ConversationChoiceNode_14 {
    class=ConversationChoiceNode
    properties {defaultChoiceDisplayText=Inquire, choiceTags {gameplayTags [{tagName=Conversation.Choice.Category.MainQuestline}]}}
   }
   FortConversationRequirement_ControllerRequirement_1 {
    class=FortConversationRequirement_ControllerRequirement
    properties {
     nodeName=Require: Is Quest Active
     requirement {
      class=FortControllerRequirement_IsQuestActive
      properties {questDefinition=/BattlePassS33_Quests/Items/QuestItems/Story/S33_Story_Confrontation/quest_S33_story_confrontation_q04a}
     }
    }
   }
  ]
 }
 Speech_C_4 {
  id=#3
  class=Speech_C
  outputs [#4]
  properties {
   softTaskIcon=&T_UI_NPC_question
   generalConfig {defaultMessage="Oh, this will be delightful. Try not to embarrass yourself."}
  }
  sub_nodes [ConversationChoiceNode_15 {class=ConversationChoiceNode, properties {defaultChoiceDisplayText=Inquire}}]
 }
 Speech_C_0 {
  id=#4
  class=Speech_C
  outputs [#5, #6, #7]
  properties {
   softTaskIcon=&T_UI_NPC_question
   generalConfig {defaultMessage="I soar through the skies graceful and light, yet I am gone with the ground in sight."}
  }
  sub_nodes [ConversationChoiceNode_0 {class=ConversationChoiceNode, properties {defaultChoiceDisplayText=Accept the Riddle}}]
 }
 Speech_C_7 {
  id=#7
  class=Speech_C
  outputs []
  properties {generalConfig {defaultMessage="I had high hopes for your humiliation, and you did not disappoint."}}
  sub_nodes [ConversationChoiceNode_17 {
    class=ConversationChoiceNode
    properties {
     defaultChoiceDisplayText=Supply Drop
     choiceTags {gameplayTags [{tagName=Athena.Quests.S33.Conversation.Story.Confrontation.Riddle.01.Failure}, {tagName=Conversation.Choice.CustomSortPriority.Highest}]}
    }
   }]
 }
 Speech_C_6 {
  id=#6
  class=Speech_C
  outputs []
  properties {generalConfig {defaultMessage="Impressive. For a human, that is."}}
  sub_nodes [ConversationChoiceNode_16 {
    class=ConversationChoiceNode
    properties {
     defaultChoiceDisplayText=Glider
     choiceTags {gameplayTags [{tagName=Athena.Quests.S33.Conversation.Story.Confrontation.Riddle.01.Success}, {tagName=Conversation.Choice.CustomSortPriority.Lowest}]}
    }
   }]
 }
 Speech_C_8 {
  id=#5
  class=Speech_C
  outputs []
  properties {generalConfig {defaultMessage="Oh, that was so wrong it's almost endearing. Almost."}}
  sub_nodes [ConversationChoiceNode_18 {
    class=ConversationChoiceNode
    properties {
     defaultChoiceDisplayText=Sun
     choiceTags {gameplayTags [{tagName=Athena.Quests.S33.Conversation.Story.Confrontation.Riddle.01.Failure}, {tagName=Conversation.Choice.CustomSortPriority.Medium}]}
    }
   }]
 }
]
