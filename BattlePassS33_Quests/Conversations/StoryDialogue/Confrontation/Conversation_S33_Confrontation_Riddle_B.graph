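# Season 33 Confrontation story NPC conversation that poses the second riddle, whose answer is the pickaxe.
# From the Riddle.02 entry point, a main-questline Inquire choice gated on quest q04a being active leads through intro lines to the riddle.
# Three terminal answer choices (Pickaxe, Grenade, Tree) each play a response and carry Success or Failure quest tags.
# Custom sort priority tags order the answers with Pickaxe highest and Tree lowest.
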ConversationDatabase /BattlePassS33_Quests/Conversations/StoryDialogue/Confrontation/Conversation_S33_Confrontation_Riddle_B
entry_points [{tag=Conversation.Entry.Season33.Story.Confrontation.Riddle.02, destinations [#1]}]
nodes [
 ConversationEntryPointNode_1 {
  id=#1
  class=ConversationEntryPointNode
  outputs [#2]
  properties {entryTag {tagName=Conversation.Entry.Season33.Story.Confrontation.Riddle.02}}
 }
 Speech_C_5 {
  id=#2
  class=Speech_C
  outputs [#3]
  properties {
   softTaskIcon=&T_UI_NPC_MainQuest
   generalConfig {defaultMessage="Well, well, a human with a question. How original."}
  }
  sub_nodes [
   ConversationChoiceNode_22 {
    class=ConversationChoiceNode
    properties {defaultChoiceDisplayText=Inquire, choiceTags {gameplayTags [{tagName=Conversation.Choice.Category.MainQuestline}]}}
   }
   FortConversationRequirement_ControllerRequirement_2 {
    class=FortConversationRequirement_ControllerRequirement
    properties {
     nodeName=Require: Is Quest Active
     requirement {
      class=FortControllerRequirement_IsQuestActive
      properties {questDefinition=/BattlePassS33_Quests/Items/QuestItems/Story/S33_Story_Confrontation/quest_S33_story_confrontation_q04a}
     }
    }
   }
  ]
 }
 Speech_C_6 {
  id=#3
  class=Speech_C
  outputs [#4]
  properties {
   softTaskIcon=&T_UI_NPC_question
   generalConfig {defaultMessage="So you think you're ready? Very well, little human, let the game begin."}
  }
  sub_nodes [ConversationChoiceNode_23 {class=ConversationChoiceNode, properties {defaultChoiceDisplayText=Inquire}}]
 }
 Speech_C_10 {
  id=#4
  class=Speech_C
  outputs [#5, #6, #7]
  properties {
   softTaskIcon=&T_UI_NPC_question
   generalConfig {defaultMessage="I stay by your side, trusty and true, In chaos or calm, I'll clear your view."}
  }
  sub_nodes [ConversationChoiceNode_27 {class=ConversationChoiceNode, properties {defaultChoiceDisplayText=Accept the Riddle}}]
 }
 Speech_C_0 {
  id=#7
  class=Speech_C
  outputs []
  properties {generalConfig {defaultMessage=A swing and a miss. Care to try embarrassing yourself again?}}
  sub_nodes [ConversationChoiceNode_0 {
    class=ConversationChoiceNode
    properties {
     defaultChoiceDisplayText=Grenade
     choiceTags {gameplayTags [{tagName=Athena.Quests.S33.Conversation.Story.Confrontation.Riddle.02.Failure}, {tagName=Conversation.Choice.CustomSortPriority.Medium}]}
    }
   }]
 }
 Speech_C_7 {
  id=#6
  class=Speech_C
  outputs []
  properties {generalConfig {defaultMessage="I must admit, I'm mildly impressed. Emphasis on mildly."}}
  sub_nodes [ConversationChoiceNode_24 {
    class=ConversationChoiceNode
    properties {
     defaultChoiceDisplayText=Pickaxe
     choiceTags {gameplayTags [{tagName=Athena.Quests.S33.Conversation.Story.Confrontation.Riddle.02.Success}, {tagName=Conversation.Choice.CustomSortPriority.Highest}]}
    }
   }]
 }
 Speech_C_9 {
  id=#5
  class=Speech_C
  outputs []
  properties {generalConfig {defaultMessage=Not even close. Are you sure you're ready for this?}}
  sub_nodes [ConversationChoiceNode_26 {
    class=ConversationChoiceNode
    properties {
     defaultChoiceDisplayText=Tree
     choiceTags {gameplayTags [{tagName=Athena.Quests.S33.Conversation.Story.Confrontation.Riddle.02.Failure}, {tagName=Conversation.Choice.CustomSortPriority.Lowest}]}
    }
   }]
 }
]
