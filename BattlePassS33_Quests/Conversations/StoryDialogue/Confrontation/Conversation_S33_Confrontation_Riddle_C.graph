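# Conversation database for the third riddle in the Season 33 Confrontation story quest, where a mocking NPC poses a riddle whose answer is Treasure Chest.
# The first dialogue choice requires the q04a quest to be active, then leads through a taunt and an accept prompt to three answer choices.
# Answers carry Success or Failure quest gameplay tags, with custom sort-priority tags ordering the correct one at lowest priority, and each ends on its own reply.
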ConversationDatabase /BattlePassS33_Quests/Conversations/StoryDialogue/Confrontation/Conversation_S33_Confrontation_Riddle_C
entry_points [{tag=Conversation.Entry.Season33.Story.Confrontation.Riddle.03, destinations [#1]}]
nodes [
 ConversationEntryPointNode_2 {
  id=#1
  class=ConversationEntryPointNode
  outputs [#2]
  properties {entryTag {tagName=Conversation.Entry.Season33.Story.Confrontation.Riddle.03}}
 }
 Speech_C_9 {
  id=#2
  class=Speech_C
  outputs [#3]
  properties {
   softTaskIcon=&T_UI_NPC_MainQuest
   generalConfig {defaultMessage="Ah, another human. Come to entertain me with your... limited capabilities?"}
  }
  sub_nodes [
   ConversationChoiceNode_0 {
    class=ConversationChoiceNode
    properties {defaultChoiceDisplayText=Inquire, choiceTags {gameplayTags [{tagName=Conversation.Choice.Category.MainQuestline}]}}
   }
   FortConversationRequirement_ControllerRequirement_0 {
    class=FortConversationRequirement_ControllerRequirement
    properties {
     nodeName=Require: Is Quest Active
     requirement {
      class=FortControllerRequirement_IsQuestActive
      properties {questDefinition=/BattlePassS33_Quests/Items/QuestItems/Story/S33_Story_Confrontation/quest_S33_story_confrontation_q04a}
     }
    }
   }
  ]
 }
 Speech_C_10 {
  id=#3
  class=Speech_C
  outputs [#4]
  properties {
   softTaskIcon=&T_UI_NPC_question
   generalConfig {defaultMessage="Hah! Eager to be humiliated? Fine, try not to embarrass your ancestors."}
  }
  sub_nodes [ConversationChoiceNode_17 {class=ConversationChoiceNode, properties {defaultChoiceDisplayText=Inquire}}]
 }
 Speech_C_14 {
  id=#4
  class=Speech_C
  outputs [#5, #6, #7]
  properties {
   softTaskIcon=&T_UI_NPC_question
   generalConfig {defaultMessage="I sing without voice, I glow without flame, To those who find me, the prize is the same"}
  }
  sub_nodes [ConversationChoiceNode_21 {class=ConversationChoiceNode, properties {defaultChoiceDisplayText=Accept the Riddle}}]
 }
 Speech_C_12 {
  id=#7
  class=Speech_C
  outputs []
  properties {generalConfig {defaultMessage="Oh my, that was awful. Even by mortal standards."}}
  sub_nodes [ConversationChoiceNode_19 {
    class=ConversationChoiceNode
    properties {
     defaultChoiceDisplayText=Ammo Crate
     choiceTags {gameplayTags [{tagName=Athena.Quests.S33.Conversation.Story.Confrontation.Riddle.03.Failure}, {tagName=Conversation.Choice.CustomSortPriority.Highest}]}
    }
   }]
 }
 Speech_C_11 {
  id=#6
  class=Speech_C
  outputs []
  properties {generalConfig {defaultMessage="Fascinating, a mortal with a brain. How delightful! What will they think of next?"}}
  sub_nodes [ConversationChoiceNode_18 {
    class=ConversationChoiceNode
    properties {
     defaultChoiceDisplayText=Treasure Chest
     choiceTags {gameplayTags [{tagName=Athena.Quests.S33.Conversation.Story.Confrontation.Riddle.03.Success}, {tagName=Conversation.Choice.CustomSortPriority.Lowest}]}
    }
   }]
 }
 Speech_C_13 {
  id=#5
  class=Speech_C
  outputs []
  properties {generalConfig {defaultMessage="Incorrect, but thanks for the laugh. Care to try embarassing yourself again?"}}
  sub_nodes [ConversationChoiceNode_20 {
    class=ConversationChoiceNode
    properties {
     defaultChoiceDisplayText=Victory Royale
     choiceTags {gameplayTags [{tagName=Athena.Quests.S33.Conversation.Story.Confrontation.Riddle.03.Failure}, {tagName=Conversation.Choice.CustomSortPriority.Medium}]}
    }
   }]
 }
]
